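; PCG example graph that, going by its name, builds a 200 m square forest from PCGPrimitives subgraphs.
; A Create_Spline_Shape subgraph produces the shape that feeds two parallel branches.
; One branch passes it to Create_Mesh_Planar for a planar mesh.
; The other passes it through Fill_Areas_Randomly, whose output drives Spawn_Assets to place the assets.
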(graph Forest_Square_200m
  (bind Subgraph_0 (Subgraph :Graph /PCGPrimitives/Primitives/Create/Create_Spline_Shape.Create_Spline_Shape))
  (bind Subgraph_4 (Subgraph :Graph /PCGPrimitives/Primitives/Fill/Fill_Areas_Randomly.Fill_Areas_Randomly))
  (bind Subgraph_5 (Subgraph :Graph /PCGPrimitives/Primitives/Spawn/Spawn_Assets.Spawn_Assets))
  (bind Subgraph_1 (Subgraph :Graph /PCGPrimitives/Primitives/Create/Create_Mesh_Planar.Create_Mesh_Planar))
  (connect Subgraph_0 Out Subgraph_1 PrimaryInput)
  (connect Subgraph_0 Out Subgraph_4 PrimaryInput)
  (connect Subgraph_4 Out Subgraph_5 PrimaryInput))
